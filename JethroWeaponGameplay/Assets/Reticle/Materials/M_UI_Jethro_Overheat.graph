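# Translucent UI material that draws an arc-shaped weapon overheat gauge for the Jethro reticle.
# Circle outlines masked by radial arcs form the progress bar, ghost bar, bracket and threshold mark, with stepped UV bands adding bracket markers.
# Overheat Amount is inverted into a Progress value, while Color A to B drives a cold/mid/heat blend and fades in a feedback arc near its maximum.
# Named reroutes organize the masks and layered color lerps.
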Material /JethroWeaponGameplay/Assets/Reticle/Materials/M_UI_Jethro_Overheat
graph {
 properties {bAutomaticallySetUsageInEditor=false, blendMode=BLEND_Translucent, materialDomain=MD_UI, usedShadingModels=Default Lit}
 nodes [
  n1 = Add(A=NamedRerouteUsage(), B=NamedRerouteUsage())
  n2 = NamedRerouteUsage()
  n3 = NamedRerouteUsage()
  n4 = NamedRerouteUsage()
  Add(A=Add(A=Add(A=@n1, B=@n2), B=Add(A=@n3, B=@n4)), B=NamedRerouteUsage())
  Cosine(None=Multiply(A=Time(), B=Constant() {r=0.162}) {constB=0.648})
  MaterialFunctionCall!(Value=Time(), Scale, Height) {materialFunction=/FortUILibrary/MaterialFunctions/Basics/MF_UI_PingPong}
  NamedRerouteDeclaration(None=MaterialFunctionCall() {materialFunction=&GetUserInterfaceUV}.Normalized UV) {name=UV, nodeColor=<rgba 0 0 0 1>}
  n5 = MaterialFunctionCall(Size=Constant() {r=0.555}, OutlineThickness=Constant() {r=0.02}) {materialFunction=&MF_UI_Circle}
  n6 = Constant() {r=0.578}
  n7 = SmoothStep(Value=MaterialFunctionCall!(UV, "Arc Start Angle (0-1)"=OneMinus(None=@n6), "Arc End Angle (0-1)"=Constant() {r=0.592}, "Fill Ratio (0-1)", Fill Clockwise) {materialFunction=/FortUILibrary/MaterialFunctions/Basics/MF_UI_RadialArc}.Centered Fill) {constMax=0.5}
  NamedRerouteDeclaration(None=Multiply(A=Multiply(A=@n5.Outline, B=@n7)) {constB=0}) {name=Threshold frame, nodeColor=<rgba 0 0 0 1>}
  n8 = MaterialFunctionCall(
   Size=ScalarParameter() {defaultValue=0.6, parameterName=Progress Bar Size}
   OutlineThickness=ScalarParameter() {defaultValue=0.02, parameterName=Progress Bar Thickness}
  ) {materialFunction=&MF_UI_Circle}
  n9 = NamedRerouteUsage()
  n10 = Constant() {r=0.5}
  n11 = SmoothStep(Value=MaterialFunctionCall(
    UV=MaterialFunctionCall(UVs=@n9, "Rotation Angle (0-1)"=@n10) {materialFunction=&CustomRotator}.Rotated Values
    "Arc Start Angle (0-1)"=ScalarParameter() {defaultValue=1, parameterName=Start Angle}
    "Arc End Angle (0-1)"=ScalarParameter() {defaultValue=0.25, parameterName=End Angle, sliderMax=1}
    "Fill Ratio (0-1)"=NamedRerouteUsage()
   ) {materialFunction=&MF_UI_RadialArc}.Radial Fill) {constMax=0.01}
  NamedRerouteDeclaration(None=Multiply(
    A=NamedRerouteDeclaration(None=@n8.Outline) {name=Circle Mask, nodeColor=<rgba 0 0 0 1>}
    B=NamedRerouteDeclaration(None=@n11) {name=Progress Mask, nodeColor=<rgba 0 0 0 1>}
   )) {name=Progress Bar, nodeColor=<rgba 0 0 0 1>}
  n12 = MaterialFunctionCall(
   Size=ScalarParameter() {defaultValue=0.6, parameterName=Progress Bar Size}
   OutlineThickness=ScalarParameter() {defaultValue=0.028, parameterName=Ghost Bar Thickness}
  ) {materialFunction=&MF_UI_Circle}
  n13 = NamedRerouteUsage()
  n14 = Constant() {r=0.5}
  n15 = SmoothStep(Value=MaterialFunctionCall(
    UV=MaterialFunctionCall(UVs=@n13, "Rotation Angle (0-1)"=@n14) {materialFunction=&CustomRotator}.Rotated Values
    "Arc Start Angle (0-1)"=ScalarParameter() {defaultValue=1, parameterName=Start Angle}
    "Arc End Angle (0-1)"=ScalarParameter() {defaultValue=0.245, parameterName=End Angle Ghost, sliderMax=1}
   ) {materialFunction=&MF_UI_RadialArc}.Radial Fill) {constMax=0.01}
  NamedRerouteDeclaration(None=Multiply(A=Multiply(A=@n12.Outline, B=@n15), B=ScalarParameter() {defaultValue=0.4, parameterName=Ghost Bar Opacity}) {constB=0.1}) {name=Ghost Bar, nodeColor=<rgba 0 0 0 1>}
  n16 = LinearInterpolate(
   A=LinearInterpolate(
    A=VectorParameter() {parameterName=Ghost Bar Color, defaultValue=<rgba 0 0 0 1>}.RGB
    B=VectorParameter() {parameterName=Bracket Color, defaultValue=<rgba 1 1 1 1>}.RGB
    Alpha=NamedRerouteUsage()
   )
   B=MaterialFunctionCall(
    A=VectorParameter() {parameterName=Cold Color, defaultValue=<rgba 1 1 1 1>}.RGB
    B=VectorParameter() {parameterName=Mid color, defaultValue=<rgba 1 0.25 0.25 1>}.RGB
    C=VectorParameter() {parameterName=Heat color, defaultValue=<rgba 1 0 0 1>}.RGB
    Alpha=ScalarParameter() {defaultValue=1, parameterName=Color A to B}
   ) {materialFunction=&3ColorBlend}.Result
   Alpha=NamedRerouteUsage()
  )
  n17 = VectorParameter() {parameterName=Threshhold Color, defaultValue=<rgba 1 1 1 1>}
  n18 = NamedRerouteUsage()
  NamedRerouteDeclaration(None=LinearInterpolate(
    A=LinearInterpolate(A=@n16, B=@n17.RGB, Alpha=@n18)
    B=VectorParameter() {parameterName=Overheat Feedback Color, defaultValue=<rgba 1 0 0 1>}.RGB
    Alpha=NamedRerouteUsage()
   ) {constAlpha=1}) {name=Colors, nodeColor=<rgba 0 0 0 1>}
  n19 = NamedRerouteUsage()
  n20 = Constant() {r=0.505}
  n21 = Constant() {r=0.755}
  n22 = Multiply(
   A=MaterialFunctionCall(Size=Constant() {r=0.66}, OutlineThickness=Constant() {r=0.015}) {materialFunction=&MF_UI_Circle}.Outline
   B=SmoothStep(Value=MaterialFunctionCall("Arc Start Angle (0-1)"=@n20, "Arc End Angle (0-1)"=@n21) {materialFunction=&MF_UI_RadialArc}.Centered Fill) {constMax=0}
  )
  NamedRerouteDeclaration(None=Multiply(A=LinearInterpolate(A=@n19, B=@n22, Alpha=@n22), B=ScalarParameter() {defaultValue=0.3, parameterName=Bracket Opacity})) {name=Bracket, nodeColor=<rgba 0 0 0 1>}
  n23 = Constant() {r=0.598}
  n24 = Constant() {r=0.04}
  n25 = MaterialFunctionCall("Arc Start Angle (0-1)"=OneMinus(None=Constant() {r=0.585}), "Arc End Angle (0-1)"=Constant() {r=0.59}) {materialFunction=&MF_UI_RadialArc}
  NamedRerouteDeclaration(None=Multiply(
    A=MaterialFunctionCall(Size=@n23, OutlineThickness=@n24) {materialFunction=&MF_UI_Circle}.Outline
    B=SmoothStep(Value=@n25.Centered Fill) {constMax=0.5}
   )) {name=Threshold, nodeColor=<rgba 0 0 0 1>}
  n26 = ScalarParameter() {defaultValue=0.81, parameterName=Marker size}
  n27 = TextureCoordinate()
  n28 = ComponentMask(None=@n27) {r=true}
  n29 = Subtract(A=Step(Y=@n26, X=@n28) {constY=0.81}, B=Step(X=@n28) {constX=0.1, constY=0.835})
  n30 = ScalarParameter() {defaultValue=0.48, parameterName=Bracket Marker Tickness}
  n31 = ComponentMask(None=@n27) {g=true}
  n32 = Subtract(A=Step(Y=@n30, X=@n31) {constY=0.48}, B=Step(X=@n31) {constX=0.1, constY=0.49})
  n33 = TextureCoordinate()
  n34 = ComponentMask(None=@n33) {r=true}
  n35 = Subtract(A=Step(Y=@n30, X=@n34) {constY=0.48}, B=Step(X=@n34) {constX=0.1, constY=0.49})
  n36 = ComponentMask(None=@n33) {g=true}
  n37 = Subtract(A=Step(Y=@n26, X=@n36) {constY=0.81}, B=Step(X=@n36) {constX=0.1, constY=0.835})
  NamedRerouteDeclaration(None=Add(A=Multiply(A=@n29, B=@n32), B=Multiply(A=@n35, B=@n37))) {name=Bracket Marker, nodeColor=<rgba 0 0 0 1>}
  n38 = MaterialFunctionCall(UVs=NamedRerouteUsage(), "Rotation Angle (0-1)"=Constant() {r=0.45}) {materialFunction=&CustomRotator}
  n39 = Constant() {r=1}
  n40 = Constant() {r=0.35}
  n41 = Multiply(
   A=MaterialFunctionCall(Size=Constant() {r=0.6}, OutlineThickness=Constant() {r=0.05}, GlowMax=Constant() {r=0.1}) {materialFunction=&MF_UI_Circle}.Outline
   B=SmoothStep(Value=MaterialFunctionCall(UV=@n38.Rotated Values, "Arc Start Angle (0-1)"=@n39, "Arc End Angle (0-1)"=@n40) {materialFunction=&MF_UI_RadialArc}.Centered Fill) {constMax=0.3}
  )
  n42 = Power(Base=SmoothStep(Value=ScalarParameter() {defaultValue=1, parameterName=Color A to B}) {constMin=0.9}) {constExponent=6}
  NamedRerouteDeclaration(None=Multiply(A=Constant() {r=0.6}, B=Multiply(A=@n41, B=@n42) {constB=0})) {name=Overheat Feedback, nodeColor=<rgba 0 0 0 1>}
  NamedRerouteDeclaration(None=OneMinus(None=ScalarParameter() {defaultValue=1, parameterName=Overheat Amount, sliderMax=1})) {name=Progress, nodeColor=<rgba 0 0 0 1>}
  NamedRerouteUsage()
 ]
}
