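# Scoped reticle extension widget for the Jethro ranged weapon, showing overheat around the crosshair.
# Construct binds weapon overheat and HUD spread events; heat ratio drives dynamic material parameters, while spread resizes a SizeBox.
# Heat updates fade the image in and set a timer, apparently to fade it out after a delay.
# Targeting changes collapse the overlay during ADS unless scoped third-person targeting is active; HandleCanting also toggles it with that mode.
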WidgetBlueprint /JethroWeaponGameplay/Assets/Reticle/WBP_Jethro_Overheat_Reticle_Scoped
animations [
 Anim_OverheatStart {
  binding_count=1
  length_seconds=0.3506
  object=&self:Anim_OverheatStart
  sequence {bindings [Image_OverheatProgress {
     id=#1
     possessed_class=Image
     tracks [{
       type=MovieSceneFloatTrack
       display_name=Render Opacity
       sections [{type=MovieSceneFloatSection, end_frame=7, start_frame=0, channels [None {type=MovieSceneScriptingFloatChannel, keys [[0,0],[7.011,1]]}]}]
      }]
    }]}
 }
 Anim_OverheatEnd {
  binding_count=1
  length_seconds=0.3492
  object=&self:Anim_OverheatEnd
  sequence {bindings [Image_OverheatProgress {
     id=#2
     possessed_class=Image
     tracks [{
       type=MovieSceneFloatTrack
       display_name=Render Opacity
       sections [{type=MovieSceneFloatSection, end_frame=6, start_frame=0, channels [None {type=MovieSceneScriptingFloatChannel, keys [[0,1],[6.983,0]]}]}]
      }]
    }]}
 }
]
graphs [
 EventGraph {nodes [
   n1 = Event OnTargetingChanged!() -> (OutputDelegate, then, bIsADS:bool) {is=title K2Node_Event}
   > n2 = Branch(Condition=@n1.bIsADS)
   > n3 = Branch(Condition=Get IsScopedThirdPersonTargeting() -> (IsScopedThirdPersonTargeting:bool))
   > SetVisibility(self=Get OverlayAll() -> (OverlayAll:Overlay Object Reference), InVisibility=SelfHitTestInvisible)
   SetVisibility(execute=@n3.else, self=Get OverlayAll(), InVisibility=Collapsed)
   SetVisibility(execute=@n2.else, self=Get OverlayAll(), InVisibility=SelfHitTestInvisible)
   Event Construct()
   > Parent: Construct()
   > Bind Event to On Spread Radius Changed(
    self:Fort HUDContext Object Reference=Get FortHUDContext!() -> (ReturnValue:Fort HUDContext Object Reference) {is=title K2Node_GetBlueprintContextEx}
    Delegate=Create Event()
   ) {is=title K2Node_AddDelegate}
   > n4 = Cast To FortWeaponRanged!(
    execute
    Object=GetEquippedWeapon!(self:Fort Weapon Reticle Extension Widget Base Object Reference) -> (ReturnValue:Fort Weapon Object Reference)
   ) -> (then, CastFailed, AsFort Weapon Ranged:Fort Weapon Ranged Object Reference) {is=title K2Node_DynamicCast}
   > n5 = Set EquippedWeapon(EquippedWeapon:Fort Weapon Ranged Object Reference=@n4) -> (Output_Get:Fort Weapon Ranged Object Reference)
   > Bind Event to On Overheat Value Changed(self:Fort Weapon Ranged Object Reference=@n5, Delegate=Create Event()) {is=title K2Node_AddDelegate}
   > SetRenderOpacity(self=Get Image_OverheatProgress() -> (Image_OverheatProgress:Image Object Reference), InOpacity=0)
   > Set IsOverheatVisualsActive(IsOverheatVisualsActive:bool=false) -> (Output_Get:bool)
   > UpdateOverheatProgress!(
    execute
    self
    CurrentValue:float=GetCurrentOverheatValue!(self:Fort Weapon Ranged Object Reference=@n5) -> (ReturnValue:float)
    MaxValue:float=GetOverheatingMaxValue!(self:Fort Weapon Ranged Object Reference=@n5) -> (ReturnValue:float)
    State:EFortWeaponOverheatState Enum=GetOverheatState!(self:Fort Weapon Ranged Object Reference=@n5) -> (ReturnValue:EFortWeaponOverheatState Enum=None)
    bIsInCooldown:bool=IsOverheated!(self:Fort Weapon Ranged Object Reference=@n5) -> (ReturnValue:bool)
   ) -> (then)
   > SetVisibility(self=Get OverlayAll(), InVisibility=HitTestInvisible)
   n6 = Get EquippedWeapon() -> (EquippedWeapon:Fort Weapon Ranged Object Reference)
   > Bind Event to On Scoped Third Person Targeting Changed(self:Fort Weapon Ranged Object Reference=@n6, Delegate=Create Event()) {is=title K2Node_AddDelegate}
   > Set IsScopedThirdPersonTargeting(IsScopedThirdPersonTargeting:bool=IsScopedThirdPersonTargetingModeEnabled!(self:Fort Weapon Ranged Object Reference=@n6) -> (ReturnValue:bool)) -> (Output_Get:bool)
   Event Destruct()
   > Parent: Destruct()
   > Unbind Event from On Spread Radius Changed(self:Fort HUDContext Object Reference=Get FortHUDContext(), Delegate=Create Event()) {is=title K2Node_RemoveDelegate}
   n7 = Get EquippedWeapon()
   > Is Valid(InputObject=@n7)
   > Unbind Event from On Overheat Value Changed(self:Fort Weapon Ranged Object Reference=@n7, Delegate=Create Event()) {is=title K2Node_RemoveDelegate}
  ]}
 HandleCanting {nodes [
   n1 = HandleCanting!() -> (then, bNewIsScopedThirdPersonTargeting:bool=~) {is=title K2Node_FunctionEntry}
   > n2 = Set IsScopedThirdPersonTargeting(IsScopedThirdPersonTargeting=@n1)
   > n3 = Branch(Condition=@n2)
   > SetVisibility(self=Get OverlayAll(), InVisibility=SelfHitTestInvisible)
   SetVisibility(execute=@n3.else, self=Get OverlayAll(), InVisibility=Collapsed)
  ]}
 OnHideOverheat {nodes [
   OnHideOverheat() {is=title K2Node_FunctionEntry}
   > Branch(Condition=Get IsOverheatVisualsActive() -> (IsOverheatVisualsActive:bool))
   > PlayAnimation(InAnimation=Get Anim_OverheatEnd() -> (Anim_OverheatEnd:Widget Animation Object Reference))
   > Set IsOverheatVisualsActive(IsOverheatVisualsActive=false)
  ]}
 OnOverheated {nodes [
   PlayAnimation()
   OnOverheated() {is=title K2Node_FunctionEntry}
   Get Anim_OverheatReached() -> (Anim_OverheatReached:Widget Animation Object Reference)
  ]}
 OnSpreadRadiusChanged {nodes [
   n1 = OnSpreadRadiusChanged!() -> (then, SpreadRadius:float=~) {is=title K2Node_FunctionEntry}
   n2 = Get SizeBox_Spread() -> (SizeBox_Spread:Size Box Object Reference)
   > SetWidthOverride(self=@n2, InWidthOverride=@n1)
   > SetHeightOverride(self=@n2, InHeightOverride=@n1)
  ]}
 UpdateOverheatProgress {nodes [
   n1 = GetDynamicMaterial()
   > SetScalarParameterValue(self=@n1, ParameterName=Progress, Value=Get PercentFill() -> (PercentFill:double))
   > n2 = SetScalarParameterValue(self=@n1, ParameterName=Glow Width, Value=Get PercentFill())
   n3 = Set Timer by Event(execute=[@n7.then, @n2.then], Delegate=Create Event(), Time=Get DelayUntilHideOverheat() -> (DelayUntilHideOverheat:float), bLooping=false)
   > Set HideOverheatTimerHandle(HideOverheatTimerHandle:Timer Handle Structure=@n3) -> (Output_Get:Timer Handle Structure)
   > n4 = Branch(Condition=Get IsOverheatVisualsActive())
   PlayAnimation(execute=@n4.else, InAnimation=Get Anim_OverheatStart() -> (Anim_OverheatStart:Widget Animation Object Reference))
   > Set IsOverheatVisualsActive(IsOverheatVisualsActive=true)
   n5 = UpdateOverheatProgress!() -> (then, CurrentValue:float=~, MaxValue:float=~, State:EFortWeaponOverheatState Enum=None, bIsInCooldown:bool=~) {is=title K2Node_FunctionEntry}
   > Set PercentFill(PercentFill:double=float / float(A=@n5, B=@n5.MaxValue) {is=title K2Node_CallFunction}) -> (Output_Get:double)
   > n6 = GetDynamicMaterial(self=Get Image_OverheatProgress())
   > SetScalarParameterValue(self=@n6, ParameterName=Overheat Amount, Value=Get PercentFill())
   > SetScalarParameterValue(self=@n6, ParameterName=Color A to B, Value=Get PercentFill())
   > n7 = SetScalarParameterValue(self=@n6, ParameterName=Glow Opacity, Value=Get PercentFill())
   SetScalarParameterValue(ParameterName=Progress, Value=0)
  ]}
]
variables [
 DelayUntilHideOverheat {type=number, default=2}
 EquippedWeapon {type=FortWeaponRanged, default=null}
 HideOverheatTimerHandle {type=TimerHandle}
 IsOverheatVisualsActive {type=boolean}
 IsScopedThirdPersonTargeting {type=boolean}
]
widgets {
 info {
  inheritedWidgetCount=1
  parentClass=/Script/FortniteUI.FortWeaponReticleExtensionWidgetBase
  rootWidgetClass=&Overlay
  widgetCount=5
 }
 widgets [
  OverlayAll {widgetClassPath=&Overlay, bIsVariable=true, widget=OverlayAll}
  Image_OverheatProgress {
   widgetClassPath=&Image
   bIsVariable=true
   parent=OverlayAll
   widget=Image_OverheatProgress
   properties {
    brush {
     bIsDynamicallyLoaded=false
     drawAs=Image
     imageType=FullColor
     mirroring=NoMirror
     resourceName=None
     resourceObject=/JethroWeaponGameplay/Assets/Reticle/Materials/MI_UI_Jethro_Overheat
     tiling=NoTile
     imageSize=<200 200>
     margin=<ltrb 0 0 0 0>
     outlineSettings {
      bUseBrushTransparency=false
      roundingType=HalfHeightRadius
      width=0
      color {colorUseRule=UseColor_Specified, specifiedColor=<rgba 0 0 0 0>}
      cornerRadii=<0 0 0 0>
     }
     tintColor {colorUseRule=UseColor_Specified, specifiedColor=<rgba 1 1 1 1>}
     uVRegion {bIsValid=false, max=<0 0>, min=<0 0>}
    }
    renderTransform {angle=0, scale=<1 1>, shear=<0 0>, translation=<40 40>}
   }
   slot {class=OverlaySlot, properties {horizontalAlignment=HAlign_Center, verticalAlignment=VAlign_Center}}
  }
  SizeBox_Spread {
   widgetClassPath=&SizeBox
   bIsVariable=true
   parent=OverlayAll
   widget=SizeBox_Spread
   properties {bOverride_HeightOverride=true, bOverride_WidthOverride=true}
   slot {class=OverlaySlot, properties {horizontalAlignment=HAlign_Center, verticalAlignment=VAlign_Center}}
  }
  CanvasPanel_55 {widgetClassPath=&CanvasPanel, parent=SizeBox_Spread, widget=CanvasPanel_55, slot {class=SizeBoxSlot}}
  TutorialHighlight {widgetClassPath=&TutorialHighlight}
 ]
}
